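# Sound cue for the discreet security camera gadget's movement, playing a direction-change sound followed almost at once by a looping move sound.
# A mixer layers a random pick of three direction-change one-shots with the move loop, which starts after a short fixed delay.
# It plays quietly in the gadgets sound class, using a movement attenuation asset and a concurrency override that stops the quietest instance.
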SoundCue /Game/Athena/Sounds/Gadgets/SecurityTurret/Security_Camera_Discreet_MoveLoop_Cue
root=Mixer_1
nodes [
 {id=Delay_2, children [WavePlayer_8], properties {delayMax=0.08, delayMin=0.08}}
 {id=Mixer_1, children [Random_3,Delay_2,null], properties {inputVolume [1,1,0.6]}}
 {id=Random_3, children [WavePlayer_9, WavePlayer_10, WavePlayer_11], properties {weights [1,1,1]}}
 {id=WavePlayer_10, properties {soundWaveAssetPtr=&/Security_Camera_MoveDirectionChange_02}}
 {id=WavePlayer_11, properties {soundWaveAssetPtr=&/Security_Camera_MoveDirectionChange_03}}
 {id=WavePlayer_8, properties {bLooping=true, soundWaveAssetPtr=&/Security_Camera_MoveLoop_01}}
 {id=WavePlayer_9, properties {soundWaveAssetPtr=&/Security_Camera_MoveDirectionChange_01}}
]
properties {
 attenuationSettings=&/Security_Camera_Discreet_Movement_Att
 bOverrideConcurrency=true
 maxDistance=1850
 soundClassObject=&Fort_Gadgets
 volumeMultiplier=0.15
 attenuationOverrides {
  bEnableFocusInterpolation=true
  bEnableListenerFocus=true
  falloffDistance=1200
  nonFocusVolumeAttenuation=0.5
  stereoSpread=500
  attenuationShapeExtents {x=0}
 }
 concurrencyOverrides {maxCount=3, resolutionRule=StopQuietest}
}
