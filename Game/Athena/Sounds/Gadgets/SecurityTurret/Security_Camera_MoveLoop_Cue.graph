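# Sound cue for the movement audio of the security turret gadget's camera.
# A mixer layers one of three random, boosted direction-change sounds, a looping movement bed that starts after a very short delay, and a beep repeated on a looped delay.
# It plays in the gadgets sound class with movement attenuation and a concurrency override that stops the quietest of excess instances.
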SoundCue /Game/Athena/Sounds/Gadgets/SecurityTurret/Security_Camera_MoveLoop_Cue
root=Mixer_1
nodes [
 {id=Delay_0, children [WavePlayer_0], properties {delayMax=6, delayMin=6}}
 {id=Delay_2, children [WavePlayer_8], properties {delayMax=0.08, delayMin=0.08}}
 {id=Looping_0, children [Delay_0]}
 {id=Mixer_1, children [Random_3,Delay_2,Looping_0], properties {inputVolume [6,1,1]}}
 {id=Random_3, children [WavePlayer_9, WavePlayer_10, WavePlayer_11], properties {weights [1,1,1]}}
 {id=WavePlayer_0, properties {soundWaveAssetPtr=&/Security_Camera_MoveLoop_Beep_01}}
 {id=WavePlayer_10, properties {soundWaveAssetPtr=&/Security_Camera_MoveDirectionChange_02}}
 {id=WavePlayer_11, properties {soundWaveAssetPtr=&/Security_Camera_MoveDirectionChange_03}}
 {id=WavePlayer_8, properties {bLooping=true, soundWaveAssetPtr=&/Security_Camera_MoveLoop_01}}
 {id=WavePlayer_9, properties {soundWaveAssetPtr=&/Security_Camera_MoveDirectionChange_01}}
]
properties {
 attenuationSettings=&/Security_Camera_Movement_Att
 bOverrideConcurrency=true
 maxDistance=2000
 soundClassObject=&Fort_Gadgets
 volumeMultiplier=0.4
 attenuationOverrides {
  bEnableFocusInterpolation=true
  bEnableListenerFocus=true
  falloffDistance=1200
  nonFocusVolumeAttenuation=0.5
  stereoSpread=500
  attenuationShapeExtents {x=0}
 }
 concurrencyOverrides {maxCount=3, resolutionRule=StopQuietest}
}
