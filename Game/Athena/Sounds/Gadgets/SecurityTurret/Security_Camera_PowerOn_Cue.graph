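# Sound cue for the power-on sound of a security camera, filed under the security turret gadget audio.
# It randomly picks one of three equally weighted wave variations and passes it through a modulator left at default settings.
# The cue is slightly pitched down and boosted in volume, uses the Gadgets sound class and a security turret sphere attenuation.
# Concurrency override allows one instance per owner, stopping the oldest.
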SoundCue /Game/Athena/Sounds/Gadgets/SecurityTurret/Security_Camera_PowerOn_Cue
root=Modulator_0
nodes [
 {id=Modulator_0, children [Random_0]}
 {id=Random_0, children [WavePlayer_0,WavePlayer_2,WavePlayer_3], properties {weights [1,1,1]}}
 {id=WavePlayer_0, properties {soundWaveAssetPtr=&/Security_Camera_PowerOn_01}}
 {id=WavePlayer_2, properties {soundWaveAssetPtr=&/Security_Camera_PowerOn_02}}
 {id=WavePlayer_3, properties {soundWaveAssetPtr=&/Security_Camera_PowerOn_03}}
]
properties {
 attenuationSettings=&/Security_Turret_Sphere_Att
 bOverrideConcurrency=true
 maxDistance=3500
 pitchMultiplier=0.9
 soundClassObject=&Fort_Gadgets
 volumeMultiplier=1.5
 concurrencyOverrides {bLimitToOwner=true, maxCount=1, resolutionRule=StopOldest}
}
