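# Sound cue for the start of the security turret's camera scan-beam loop, in the gadgets sound class.
# A random node picks one of three equally weighted start wave variants, with no looping set in the cue.
# Spatialization comes from a shared turret sphere attenuation asset, and a concurrency override caps simultaneous instances by stopping the quietest.
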SoundCue /Game/Athena/Sounds/Gadgets/SecurityTurret/Security_Camera_ScanBeam_LoopStart_Cue
root=Random_0
nodes [
 {id=Random_0, children [WavePlayer_8, WavePlayer_9, WavePlayer_10], properties {weights [1,1,1]}}
 {id=WavePlayer_10, properties {soundWaveAssetPtr=&/Security_Camera_ScanBeam_Start_03}}
 {id=WavePlayer_8, properties {soundWaveAssetPtr=&/Security_Camera_ScanBeam_Start_01}}
 {id=WavePlayer_9, properties {soundWaveAssetPtr=&/Security_Camera_ScanBeam_Start_02}}
]
properties {
 attenuationSettings=&/Security_Turret_Sphere_Att
 bOverrideConcurrency=true
 maxDistance=3500
 soundClassObject=&Fort_Gadgets
 volumeMultiplier=1.2
 attenuationOverrides {
  bEnableFocusInterpolation=true
  bEnableListenerFocus=true
  falloffDistance=1200
  nonFocusVolumeAttenuation=0.5
  stereoSpread=500
  attenuationShapeExtents {x=0}
 }
 concurrencyOverrides {maxCount=3, resolutionRule=StopQuietest}
}
