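# Looping sound cue for the scan beam of a security camera in the security turret gadget audio, routed to the gadgets sound class.
# It mixes narrow and wide looping layers, with the narrow one quieter.
# Each layer passes through its own matching cone attenuation node, so the two can respond differently to listener position.
# A concurrency override caps it at three instances, stopping the quietest.
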SoundCue /Game/Athena/Sounds/Gadgets/SecurityTurret/Security_Camera_ScanBeam_Loop_Cue
root=Mixer_0
nodes [
 {
  id=Attenuation_0
  children [WavePlayer_0]
  properties {attenuationSettings=&/Security_Turret_Sight_Beam_Cone_Wide_Att}
 }
 {
  id=Attenuation_1
  children [WavePlayer_1]
  properties {attenuationSettings=&/Security_Turret_Sight_Beam_Cone_Narrow_Att}
 }
 {id=Mixer_0, children [Attenuation_1, Attenuation_0], properties {inputVolume [0.6,1]}}
 {
  id=WavePlayer_0
  properties {bLooping=true, soundWaveAssetPtr=&/Security_Camera_ScanBeam_Loop_01_Wide}
 }
 {
  id=WavePlayer_1
  properties {bLooping=true, soundWaveAssetPtr=&/Security_Camera_ScanBeam_Loop_01_Narrow}
 }
]
properties {
 bOverrideConcurrency=true
 maxDistance=2550
 soundClassObject=&Fort_Gadgets
 attenuationOverrides {
  bEnableFocusInterpolation=true
  bEnableListenerFocus=true
  falloffDistance=1200
  nonFocusVolumeAttenuation=0.5
  stereoSpread=500
  attenuationShapeExtents {x=0}
 }
 concurrencyOverrides {maxCount=3, resolutionRule=StopQuietest}
}
