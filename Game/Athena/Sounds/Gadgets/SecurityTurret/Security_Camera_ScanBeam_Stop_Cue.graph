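# Sound cue for the security camera scan beam stop sound, filed with the security turret gadget audio.
# A single wave player plays one stop sample directly, with no randomization or modulation nodes.
# It uses the security turret's spherical attenuation and the gadgets sound class.
# Overridden concurrency limits it to one instance per owner, stopping the oldest when a new one plays.
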SoundCue /Game/Athena/Sounds/Gadgets/SecurityTurret/Security_Camera_ScanBeam_Stop_Cue
root=WavePlayer_0
nodes [{id=WavePlayer_0, properties {soundWaveAssetPtr=&/Security_Camera_ScanBeam_Stop_01}}]
properties {
 attenuationSettings=&/Security_Turret_Sphere_Att
 bOverrideConcurrency=true
 maxDistance=3500
 soundClassObject=&Fort_Gadgets
 volumeMultiplier=1
 concurrencyOverrides {bLimitToOwner=true, maxCount=1, resolutionRule=StopOldest}
}
