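# Sound cue for the target beam of a security camera or turret gadget, routed to the gadgets sound class.
# A mixer layers a start one-shot, slightly pitched down with sphere attenuation, over an unmodulated looping beam sound with wide sight-beam cone attenuation.
# Per-branch attenuation gives each layer its own spatial shape.
# Concurrency is capped at a few instances, stopping the quietest.
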SoundCue /Game/Athena/Sounds/Gadgets/SecurityTurret/Security_Camera_TargetBeam_Loop_Cue
root=Mixer_0
nodes [
 {
  id=Attenuation_1
  children [Modulator_0]
  properties {
   attenuationSettings=&/Security_Turret_Sight_Beam_Cone_Wide_Att
   attenuationOverrides {falloffDistance=500, attenuationShapeExtents {x=0}}
  }
 }
 {
  id=Attenuation_2
  children [Modulator_1]
  properties {
   attenuationSettings=&/Security_Turret_Sphere_Att
   attenuationOverrides {falloffDistance=500, attenuationShapeExtents {x=0}}
  }
 }
 {id=Mixer_0, children [Attenuation_2, Attenuation_1, null], properties {inputVolume [0.9,1,1]}}
 {id=Modulator_0, children [WavePlayer_6], properties {pitchMax=1, pitchMin=1, volumeMax=1, volumeMin=1}}
 {id=Modulator_1, children [WavePlayer_7], properties {pitchMax=0.95}}
 {id=WavePlayer_6, properties {bLooping=true, soundWaveAssetPtr=&/Security_Camera_TargetBeam_Loop_01}}
 {id=WavePlayer_7, properties {soundWaveAssetPtr=&/Security_Camera_TargetBeam_Start_01}}
]
properties {
 bOverrideConcurrency=true
 maxDistance=3500
 soundClassObject=&Fort_Gadgets
 volumeMultiplier=0.85
 attenuationOverrides {
  bEnableFocusInterpolation=true
  bEnableListenerFocus=true
  falloffDistance=1200
  nonFocusVolumeAttenuation=0.5
  stereoSpread=500
  attenuationShapeExtents {x=0}
 }
 concurrencyOverrides {maxCount=3, resolutionRule=StopQuietest}
}
