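# Sound cue for the security camera gadget losing its target, layering a mechanical stinger with a delayed voice line.
# The effects branch mixes the target-lost sound with a scan-beam stop through the turret's sphere attenuation.
# The voice branch waits briefly, then plays one of four equally weighted lines through its own vox attenuation.
# It uses the gadget sound class and limits playback to one instance per owner, stopping the oldest.
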SoundCue /Game/Athena/Sounds/Gadgets/SecurityTurret/Security_Camera_Target_Lost_Cue
root=Mixer_0
nodes [
 {id=Attenuation_0, children [Mixer_1], properties {attenuationSettings=&/Security_Turret_Sphere_Att}}
 {id=Attenuation_1, children [Delay_0], properties {attenuationSettings=&/Security_Turret_Vox_Att}}
 {id=Delay_0, children [Random_0], properties {delayMax=0.6, delayMin=0.6}}
 {id=Mixer_0, children [Attenuation_0, Attenuation_1], properties {inputVolume [1,0.8]}}
 {id=Mixer_1, children [WavePlayer_4,WavePlayer_0], properties {inputVolume [0.8,0.6]}}
 {id=Random_0, children [WavePlayer_1,WavePlayer_2,WavePlayer_3,WavePlayer_6], properties {weights [1,1,1,1]}}
 {id=WavePlayer_0, properties {soundWaveAssetPtr=&/Security_Camera_ScanBeam_Stop_01}}
 {id=WavePlayer_1, properties {soundWaveAssetPtr=&/Security_Camera_Vox_Target_Lost_01}}
 {id=WavePlayer_2, properties {soundWaveAssetPtr=&/Security_Camera_Vox_Target_Lost_02}}
 {id=WavePlayer_3, properties {soundWaveAssetPtr=&/Security_Camera_Vox_Target_Lost_03}}
 {id=WavePlayer_4, properties {soundWaveAssetPtr=&/Security_Camera_Target_Lost_02}}
 {id=WavePlayer_6, properties {soundWaveAssetPtr=&/Security_Camera_Vox_Target_Lost_04}}
]
properties {
 bOverrideConcurrency=true
 maxDistance=3500
 soundClassObject=&Fort_Gadgets
 volumeMultiplier=1.25
 concurrencyOverrides {bLimitToOwner=true, maxCount=1, resolutionRule=StopOldest}
}
