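# Sound cue for a Security Turret impact on a player, built from shared hit-notification sounds.
# A mixer layers a randomized body hit-marker with a slightly delayed, randomized receive UI tick, each through its own modulator.
# The cue raises overall pitch, uses the 2D receive hit-notification sound class with instigator hit-marker attenuation, and overrides concurrency to stop the oldest instance beyond a small limit.
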SoundCue /Game/Athena/Sounds/Gadgets/SecurityTurret/Security_Turret_ChargPlayerImpact_Cue
root=Mixer_0
nodes [
 {id=Delay_1, children [Modulator_2], properties {delayMax=0.05, delayMin=0.05}}
 {id=Mixer_0, children [Modulator_0,Delay_1], properties {inputVolume [1,1]}}
 {id=Modulator_0, children [Random_0]}
 {id=Modulator_2, children [Random_2], properties {pitchMax=1.1, volumeMax=1, volumeMin=1}}
 {id=Random_0, children [WavePlayer_4,WavePlayer_5,WavePlayer_6], properties {bShouldExcludeFromBranchCulling=true, weights [1,1,1]}}
 {id=Random_2, children [WavePlayer_12, WavePlayer_13, WavePlayer_14], properties {bShouldExcludeFromBranchCulling=true, weights [1,1,1]}}
 {id=WavePlayer_12, properties {soundWaveAssetPtr=/Game/Athena/Sounds/HitNotifications/HitMarker_Receive_UI_Tick_01}}
 {id=WavePlayer_13, properties {soundWaveAssetPtr=/Game/Athena/Sounds/HitNotifications/HitMarker_Receive_UI_Tick_02}}
 {id=WavePlayer_14, properties {soundWaveAssetPtr=/Game/Athena/Sounds/HitNotifications/HitMarker_Receive_UI_Tick_03}}
 {id=WavePlayer_4, properties {soundWaveAssetPtr=/Game/Athena/Sounds/HitNotifications/HitMarker_Body_01}}
 {id=WavePlayer_5, properties {soundWaveAssetPtr=/Game/Athena/Sounds/HitNotifications/HitMarker_Body_02}}
 {id=WavePlayer_6, properties {soundWaveAssetPtr=/Game/Athena/Sounds/HitNotifications/HitMarker_Body_03}}
]
properties {
 attenuationSettings=/Game/Athena/Sounds/HitNotifications/Fort_HitMarker_Instigate_Att
 bOverrideConcurrency=true
 maxDistance=2097152
 pitchMultiplier=1.5
 soundClassObject=/Game/Sounds/FortSoundClassesAndModes/Fort_2DHitNotifications_Receive_PSM
 concurrencyOverrides {maxCount=3, resolutionRule=StopOldest}
}
