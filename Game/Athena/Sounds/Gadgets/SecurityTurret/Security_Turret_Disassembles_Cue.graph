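# Sound cue for the disassembly sound of the Athena security turret gadget.
# A fixed half-second delay feeds a single wave player for the disassemble sound.
# It uses the turret's spherical attenuation and the gadget sound class at reduced volume.
# An overridden concurrency rule allows one instance per owner, stopping the oldest when a new one starts.
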SoundCue /Game/Athena/Sounds/Gadgets/SecurityTurret/Security_Turret_Disassembles_Cue
root=Delay_0
nodes [
 {id=Delay_0, children [WavePlayer_1], properties {delayMax=0.5, delayMin=0.5}}
 {id=WavePlayer_1, properties {soundWaveAssetPtr=&/Security_Turret_Disassembles_01}}
]
properties {
 attenuationSettings=&/Security_Turret_Sphere_Att
 bOverrideConcurrency=true
 maxDistance=3500
 soundClassObject=&Fort_Gadgets
 volumeMultiplier=0.5
 concurrencyOverrides {bLimitToOwner=true, maxCount=1, resolutionRule=StopOldest}
}
