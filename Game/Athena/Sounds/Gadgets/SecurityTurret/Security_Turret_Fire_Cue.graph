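# Sound cue for the Security Turret gadget's firing sound, layering close, medium and distant heavy-fire variations.
# Each layer randomly picks a wave and uses its own sentry attenuation, with medium and distant mixed under a distant-fire sound class.
# The close layer uses a third-person fire class with slight pitch and volume modulation, while fixed modulators boost the other layers.
# Concurrency is limited per owner, stopping the oldest shot.
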SoundCue /Game/Athena/Sounds/Gadgets/SecurityTurret/Security_Turret_Fire_Cue
root=Mixer_1
nodes [
 {id=Attenuation_118, children [Random_1], properties {attenuationSettings=/Game/Sounds/Attenuation/Fort_Weapon_Projectile_Close_Sentry_Att}}
 {
  id=Attenuation_119
  children [Modulator_0]
  properties {attenuationSettings=/Game/Sounds/Attenuation/Fort_Weapon_Projectile_Medium_Sentry_Att}
 }
 {
  id=Attenuation_120
  children [Modulator_1]
  properties {attenuationSettings=/Game/Sounds/Attenuation/Fort_Weapon_Projectile_Distant_Sentry_Att}
 }
 {id=Mixer_1, children [Modulator_2, SoundClass_76], properties {inputVolume [1,0.6]}}
 {id=Mixer_4, children [Attenuation_119, Attenuation_120], properties {inputVolume [1,1]}}
 {id=Modulator_0, children [Random_4], properties {volumeMax=1.7, volumeMin=1.7}}
 {id=Modulator_1, children [Random_5], properties {volumeMax=1.8, volumeMin=1.8}}
 {id=Modulator_2, children [SoundClass_77], properties {pitchMax=1.1, pitchMin=0.9, volumeMax=1.1, volumeMin=0.9}}
 {
  id=Random_1
  children [WavePlayer_4, WavePlayer_5, WavePlayer_6, WavePlayer_15, WavePlayer_16]
  properties {bShouldExcludeFromBranchCulling=true, weights [1,1,1,1,1]}
 }
 {
  id=Random_4
  children [WavePlayer_17, WavePlayer_18, WavePlayer_19, WavePlayer_20]
  properties {bShouldExcludeFromBranchCulling=true, weights [1,1,1,1]}
 }
 {id=Random_5, children [WavePlayer_21, WavePlayer_22, WavePlayer_23], properties {bShouldExcludeFromBranchCulling=true, weights [1,1,1]}}
 {id=SoundClass_76, children [Mixer_4], properties {soundClassOverride=&Fort_WeaponRanged_Fire_Distant}}
 {
  id=SoundClass_77
  children [Attenuation_118]
  properties {soundClassOverride=&Fort_WeaponRanged_Fire_3P_PSM}
 }
 {id=WavePlayer_15, properties {soundWaveAssetPtr=&/Security_Turret_Fire_Heavy_Close_02}}
 {id=WavePlayer_16, properties {soundWaveAssetPtr=&/Security_Turret_Fire_Heavy_Close_01}}
 {id=WavePlayer_17, properties {soundWaveAssetPtr=&/Security_Turret_Fire_Heavy_Medium_04}}
 {id=WavePlayer_18, properties {soundWaveAssetPtr=&/Security_Turret_Fire_Heavy_Medium_03}}
 {id=WavePlayer_19, properties {soundWaveAssetPtr=&/Security_Turret_Fire_Heavy_Medium_02}}
 {id=WavePlayer_20, properties {soundWaveAssetPtr=&/Security_Turret_Fire_Heavy_Medium_01}}
 {id=WavePlayer_21, properties {soundWaveAssetPtr=&/Security_Turret_Fire_Heavy_Distant_01}}
 {id=WavePlayer_22, properties {soundWaveAssetPtr=&/Security_Turret_Fire_Heavy_Distant_02}}
 {id=WavePlayer_23, properties {soundWaveAssetPtr=&/Security_Turret_Fire_Heavy_Distant_03}}
 {id=WavePlayer_4, properties {soundWaveAssetPtr=&/Security_Turret_Fire_Heavy_Close_03}}
 {id=WavePlayer_5, properties {soundWaveAssetPtr=&/Security_Turret_Fire_Heavy_Close_04}}
 {id=WavePlayer_6, properties {soundWaveAssetPtr=&/Security_Turret_Fire_Heavy_Close_05}}
]
properties {
 bDebug=true
 bOverrideConcurrency=true
 maxDistance=14000
 volumeMultiplier=0.4
 concurrencyOverrides {bLimitToOwner=true, maxCount=3, resolutionRule=StopOldest}
}
