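# Sound cue for Security Turret impacts in Athena, routed to the gadgets sound class with bullet-impact attenuation.
# A random node picks one of five equally weighted impact waves, which passes through a modulator for variation and a short fixed delay.
# The cue raises overall pitch and lowers volume, and its concurrency override uses a prevent-new resolution rule.
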SoundCue /Game/Athena/Sounds/Gadgets/SecurityTurret/Security_Turret_Impact_Cue
root=Delay_0
nodes [
 {id=Delay_0, children [Modulator_1], properties {delayMax=0.05, delayMin=0.05}}
 {id=Modulator_1, children [Random_3], properties {pitchMax=1}}
 {id=Random_3, children [WavePlayer_13, WavePlayer_14, WavePlayer_15, WavePlayer_16, WavePlayer_17], properties {weights [1,1,1,1,1]}}
 {id=WavePlayer_13, properties {soundWaveAssetPtr=&/Security_Turret_Impact_02}}
 {id=WavePlayer_14, properties {soundWaveAssetPtr=&/Security_Turret_Impact_01}}
 {id=WavePlayer_15, properties {soundWaveAssetPtr=&/Security_Turret_Impact_03}}
 {id=WavePlayer_16, properties {soundWaveAssetPtr=&/Security_Turret_Impact_04}}
 {id=WavePlayer_17, properties {soundWaveAssetPtr=&/Security_Turret_Impact_05}}
]
properties {
 attenuationSettings=/Game/Sounds/Attenuation/Fort_Bullet_Impact_Athena_Att
 maxDistance=4500
 pitchMultiplier=1.5
 soundClassObject=&Fort_Gadgets
 volumeMultiplier=0.6
 concurrencyOverrides {resolutionRule=PreventNew}
}
