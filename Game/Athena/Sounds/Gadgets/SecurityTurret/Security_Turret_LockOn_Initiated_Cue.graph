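# Sound cue for the security turret gadget's lock-on initiated sound in Athena.
# A single wave player feeds the lock-on wave through a modulator at the root, with reduced cue volume, in the gadgets sound class.
# It references a spherical no-occlusion attenuation asset.
# Its own concurrency override caps simultaneous instances and stops the quietest when the limit is exceeded.
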SoundCue /Game/Athena/Sounds/Gadgets/SecurityTurret/Security_Turret_LockOn_Initiated_Cue
root=Modulator_1
nodes [
 {id=Modulator_1, children [WavePlayer_7]}
 {id=WavePlayer_7, properties {soundWaveAssetPtr=&/Security_Turret_LockOn_Initiated}}
]
properties {
 attenuationSettings=&/Security_Turret_Sphere_NoOcclusion_Att
 bOverrideConcurrency=true
 maxDistance=5000
 soundClassObject=&Fort_Gadgets
 volumeMultiplier=0.3
 attenuationOverrides {
  bEnableFocusInterpolation=true
  bEnableListenerFocus=true
  falloffDistance=1200
  nonFocusVolumeAttenuation=0.5
  stereoSpread=500
  attenuationShapeExtents {x=0}
 }
 concurrencyOverrides {maxCount=2, resolutionRule=StopQuietest}
}
