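# Sound cue for the security turret gadget's movement, pairing a random start one-shot with a looping move sound.
# A mixer plays one of three equally weighted MoveStart waves alongside the move loop, which a short fixed delay holds back so the start leads into it.
# It uses turret movement attenuation and the gadgets sound class, with a concurrency override that caps instances and stops the quietest.
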SoundCue /Game/Athena/Sounds/Gadgets/SecurityTurret/Security_Turret_Move_Loop_Cue
root=Mixer_1
nodes [
 {id=Delay_2, children [WavePlayer_1], properties {delayMax=0.08, delayMin=0.08}}
 {id=Mixer_1, children [Random_2,Delay_2], properties {inputVolume [0.7,0.7]}}
 {id=Random_2, children [WavePlayer_6, WavePlayer_7, WavePlayer_12], properties {weights [1,1,1]}}
 {id=WavePlayer_1, properties {bLooping=true, soundWaveAssetPtr=&/Security_Turret_MoveLoop_01}}
 {id=WavePlayer_12, properties {soundWaveAssetPtr=&/Security_Turret_MoveStart_03}}
 {id=WavePlayer_6, properties {soundWaveAssetPtr=&/Security_Turret_MoveStart_01}}
 {id=WavePlayer_7, properties {soundWaveAssetPtr=&/Security_Turret_MoveStart_02}}
]
properties {
 attenuationSettings=&/Security_Turret_Movement_Att
 bOverrideConcurrency=true
 maxDistance=2500
 soundClassObject=&Fort_Gadgets
 volumeMultiplier=0.6
 attenuationOverrides {
  bEnableFocusInterpolation=true
  bEnableListenerFocus=true
  falloffDistance=1200
  nonFocusVolumeAttenuation=0.5
  stereoSpread=500
  attenuationShapeExtents {x=0}
 }
 concurrencyOverrides {maxCount=3, resolutionRule=StopQuietest}
}
