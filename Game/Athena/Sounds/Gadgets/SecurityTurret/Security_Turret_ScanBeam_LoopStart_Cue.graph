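# Sound cue for the start of the security turret's scan beam loop, a gadget sound.
# A Random node picks one of three start waves with equal weight and feeds it to the output at reduced volume, with no looping node in the cue.
# It uses the turret's shared sphere attenuation asset and the gadgets sound class, with its own concurrency limit that stops the quietest instance.
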SoundCue /Game/Athena/Sounds/Gadgets/SecurityTurret/Security_Turret_ScanBeam_LoopStart_Cue
root=Random_2
nodes [
 {id=Random_2, children [WavePlayer_21, WavePlayer_22, WavePlayer_23], properties {weights [1,1,1]}}
 {id=WavePlayer_21, properties {soundWaveAssetPtr=&/Security_Turret_ScanBeam_Start_03}}
 {id=WavePlayer_22, properties {soundWaveAssetPtr=&/Security_Turret_ScanBeam_Start_01}}
 {id=WavePlayer_23, properties {soundWaveAssetPtr=&/Security_Turret_ScanBeam_Start_02}}
]
properties {
 attenuationSettings=&/Security_Turret_Sphere_Att
 bOverrideConcurrency=true
 maxDistance=3500
 soundClassObject=&Fort_Gadgets
 volumeMultiplier=0.3
 attenuationOverrides {
  bEnableFocusInterpolation=true
  bEnableListenerFocus=true
  falloffDistance=1200
  nonFocusVolumeAttenuation=0.5
  stereoSpread=500
  attenuationShapeExtents {x=0}
 }
 concurrencyOverrides {maxCount=3, resolutionRule=StopQuietest}
}
