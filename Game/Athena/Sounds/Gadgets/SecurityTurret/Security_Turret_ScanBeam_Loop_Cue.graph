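# Sound cue for the looping scan beam of the Security Turret gadget.
# A mixer blends two looping waves, narrow and wide, each through its own matching sight-beam cone attenuation, with the narrow layer slightly quieter.
# The cue routes to the gadgets sound class, uses listener focus to lower non-focused volume, and overrides concurrency to cap instances, stopping the quietest.
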SoundCue /Game/Athena/Sounds/Gadgets/SecurityTurret/Security_Turret_ScanBeam_Loop_Cue
root=Mixer_1
nodes [
 {
  id=Attenuation_4
  children [WavePlayer_26]
  properties {attenuationSettings=&/Security_Turret_Sight_Beam_Cone_Wide_Att}
 }
 {
  id=Attenuation_5
  children [WavePlayer_25]
  properties {
   attenuationSettings=&/Security_Turret_Sight_Beam_Cone_Narrow_Att
   attenuationOverrides {falloffDistance=500, attenuationShapeExtents {x=0}}
  }
 }
 {id=Mixer_1, children [Attenuation_5, Attenuation_4], properties {inputVolume [0.75,1]}}
 {
  id=WavePlayer_25
  properties {bLooping=true, soundWaveAssetPtr=&/Security_Turret_ScanBeam_Loop_01_Narrow}
 }
 {
  id=WavePlayer_26
  properties {bLooping=true, soundWaveAssetPtr=&/Security_Turret_ScanBeam_Loop_01_Wide}
 }
]
properties {
 bOverrideConcurrency=true
 maxDistance=2550
 soundClassObject=&Fort_Gadgets
 volumeMultiplier=1
 attenuationOverrides {
  bEnableFocusInterpolation=true
  bEnableListenerFocus=true
  falloffDistance=1200
  nonFocusVolumeAttenuation=0.5
  stereoSpread=500
  attenuationShapeExtents {x=0}
 }
 concurrencyOverrides {maxCount=3, resolutionRule=StopQuietest}
}
