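# Sound cue for the security turret gadget's target beam, in the gadgets sound class at reduced volume.
# A mixer plays the beam start wave at full level alongside the loop wave at slightly lower level.
# Each wave has its own attenuation node: a narrow sight-beam cone for the start, a wide one for the loop.
# Overridden concurrency caps simultaneous instances at a few, stopping the quietest.
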SoundCue /Game/Athena/Sounds/Gadgets/SecurityTurret/Security_Turret_TargetBeam_Loop_Cue
root=Mixer_1
nodes [
 {
  id=Attenuation_2
  children [WavePlayer_0]
  properties {
   attenuationSettings=&/Security_Turret_Sight_Beam_Cone_Wide_Att
   attenuationOverrides {falloffDistance=500, attenuationShapeExtents {x=0}}
  }
 }
 {
  id=Attenuation_3
  children [WavePlayer_4]
  properties {attenuationSettings=&/Security_Turret_Sight_Beam_Cone_Narrow_Att}
 }
 {id=Mixer_1, children [Attenuation_3, Attenuation_2], properties {inputVolume [1,0.8]}}
 {id=WavePlayer_0, properties {soundWaveAssetPtr=&/Security_Turret_TargetBeam_Loop_03}}
 {id=WavePlayer_4, properties {soundWaveAssetPtr=&/Security_Turret_TargetBeam_Start_01}}
]
properties {
 bOverrideConcurrency=true
 maxDistance=2550
 soundClassObject=&Fort_Gadgets
 volumeMultiplier=0.3
 attenuationOverrides {
  bEnableFocusInterpolation=true
  bEnableListenerFocus=true
  falloffDistance=1200
  nonFocusVolumeAttenuation=0.5
  stereoSpread=500
  attenuationShapeExtents {x=0}
 }
 concurrencyOverrides {maxCount=3, resolutionRule=StopQuietest}
}
