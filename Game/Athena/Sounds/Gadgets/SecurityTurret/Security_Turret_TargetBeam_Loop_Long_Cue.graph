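# Sound cue for the long-range looping target beam of the security turret gadget, in the Fort_Gadgets sound class.
# A mixer plays a one-shot turret beam start sound alongside a quieter looping wave borrowed from the security camera, which a modulator pitches up by a fixed amount.
# It uses a long cone sight-beam attenuation asset and overrides concurrency to cap instances, stopping the quietest.
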SoundCue /Game/Athena/Sounds/Gadgets/SecurityTurret/Security_Turret_TargetBeam_Loop_Long_Cue
root=Mixer_1
nodes [
 {id=Mixer_1, children [WavePlayer_4,Modulator_0], properties {inputVolume [1,0.8]}}
 {id=Modulator_0, children [WavePlayer_6], properties {pitchMax=1.5, pitchMin=1.5, volumeMax=1, volumeMin=1}}
 {id=WavePlayer_4, properties {soundWaveAssetPtr=&/Security_Turret_TargetBeam_Start_01}}
 {id=WavePlayer_6, properties {bLooping=true, soundWaveAssetPtr=&/Security_Camera_TargetBeam_Loop_01}}
]
properties {
 attenuationSettings=&/Security_Turret_Sight_Beam_Cone_Long_Att
 bDebug=true
 bOverrideConcurrency=true
 maxDistance=4250
 soundClassObject=&Fort_Gadgets
 volumeMultiplier=0.6
 attenuationOverrides {
  bEnableFocusInterpolation=true
  bEnableListenerFocus=true
  falloffDistance=1200
  nonFocusVolumeAttenuation=0.5
  stereoSpread=500
  attenuationShapeExtents {x=0}
 }
 concurrencyOverrides {maxCount=3, resolutionRule=StopQuietest}
}
