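# Sound cue for the security turret gadget's target beam stopping, in the Fort_Gadgets sound class.
# A single wave player feeds a modulator that fixes pitch and boosts volume, under a reduced cue volume multiplier.
# It uses the turret's sphere attenuation and overrides concurrency to allow one instance per owner, stopping the oldest.
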SoundCue /Game/Athena/Sounds/Gadgets/SecurityTurret/Security_Turret_TargetBeam_Stop_Cue
root=Modulator_0
nodes [
 {id=Modulator_0, children [WavePlayer_0], properties {pitchMax=1, pitchMin=1, volumeMax=1.7, volumeMin=1.7}}
 {id=WavePlayer_0, properties {soundWaveAssetPtr=&/Security_Turret_TargetBeam_Stop_01}}
]
properties {
 attenuationSettings=&/Security_Turret_Sphere_Att
 bOverrideConcurrency=true
 maxDistance=3500
 soundClassObject=&Fort_Gadgets
 volumeMultiplier=0.35
 concurrencyOverrides {bLimitToOwner=true, maxCount=1, resolutionRule=StopOldest}
}
