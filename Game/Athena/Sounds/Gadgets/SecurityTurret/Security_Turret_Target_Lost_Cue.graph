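# Sound cue for the security turret gadget's target-lost sound, played at reduced volume through the gadgets sound class.
# A single wave player feeds a modulator that applies a fixed, slightly lowered pitch with no random variation.
# It uses the turret's sphere attenuation asset for spatialization and falloff.
# A concurrency override allows one instance per owner, stopping the oldest when a new one plays.
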SoundCue /Game/Athena/Sounds/Gadgets/SecurityTurret/Security_Turret_Target_Lost_Cue
root=Modulator_0
nodes [
 {id=Modulator_0, children [WavePlayer_2], properties {pitchMax=0.9, pitchMin=0.9}}
 {id=WavePlayer_2, properties {soundWaveAssetPtr=&/Security_Turret_Target_Lost_02}}
]
properties {
 attenuationSettings=&/Security_Turret_Sphere_Att
 bOverrideConcurrency=true
 maxDistance=3500
 soundClassObject=&Fort_Gadgets
 volumeMultiplier=0.5
 concurrencyOverrides {bLimitToOwner=true, maxCount=1, resolutionRule=StopOldest}
}
