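# Security turret sound cue for spotting a target, apparently the large-attenuation variant, in the Fort_Gadgets sound class.
# A Mixer layers a fixed target-spotted wave with one of three equally weighted security camera move-start waves, with the camera layer boosted slightly.
# Inline attenuation overrides replace the referenced sphere attenuation, giving a long falloff with distance low-pass filtering.
# Concurrency allows one instance per owner and blocks new plays while it sounds.
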SoundCue /Game/Athena/Sounds/Gadgets/SecurityTurret/Security_Turret_Target_Spotted_LargeAtt_Cue
root=Mixer_0
nodes [
 {id=Mixer_0, children [WavePlayer_5,Random_1], properties {inputVolume [1,1.2]}}
 {id=Random_1, children [WavePlayer_3,WavePlayer_4,WavePlayer_6], properties {weights [1,1,1]}}
 {id=WavePlayer_3, properties {soundWaveAssetPtr=&/Security_Camera_MoveStart_01}}
 {id=WavePlayer_4, properties {soundWaveAssetPtr=&/Security_Camera_MoveStart_02}}
 {id=WavePlayer_5, properties {soundWaveAssetPtr=&/Security_Turret_Target_Spotted_03}}
 {id=WavePlayer_6, properties {soundWaveAssetPtr=&/Security_Camera_MoveStart_03}}
]
properties {
 attenuationSettings=&/Security_Turret_Sphere_Att
 bOverrideAttenuation=true
 bOverrideConcurrency=true
 maxDistance=5500
 soundClassObject=&Fort_Gadgets
 volumeMultiplier=1.2
 attenuationOverrides {
  bAttenuateWithLPF=true
  bEnableFocusInterpolation=true
  falloffDistance=5000
  lPFFrequencyAtMax=8000
  lPFRadiusMax=3500
  lPFRadiusMin=500
  nonFocusVolumeAttenuation=0.5
  occlusionLowPassFilterFrequency=10000
  occlusionVolumeAttenuation=0.8
  stereoSpread=100
  attenuationShapeExtents {x=500}
 }
 concurrencyOverrides {bLimitToOwner=true, maxCount=1, resolutionRule=PreventNew}
}
